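# Sound cue for the lock-on copter warning of the MissileBattery launcher (MistyBop).
# A mixer starts a one-shot start sound and a looping warning together, with the loop slightly quieter.
# It uses gadget-activate attenuation and an important in-game mission sound class.
# A concurrency override allows one instance at a time, stopping the oldest when a new one plays.
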SoundCue /FNGameplayCues/Sounds/Fort_Weapons/Launchers/MissileBattery/MistyBop_LockOnCopter_Cue
root=Mixer_0
nodes [
 {id=Mixer_0, children [WavePlayer_1,WavePlayer_2], properties {inputVolume [1,0.8]}}
 {id=WavePlayer_1, properties {soundWaveAssetPtr=&/MistyBop_LockOnCopter_Start_02}}
 {
  id=WavePlayer_2
  properties {bLooping=true, soundWaveAssetPtr=&/MistyBop_LockOnCopter_Warning_Loop_02}
 }
]
properties {
 attenuationSettings=&Fort_GadgetActivate_Att
 bOverrideConcurrency=true
 maxDistance=4000
 soundClassObject=&Fort_Mission_Ingame_Important_PSM
 volumeMultiplier=0.8
 concurrencyOverrides {maxCount=1, resolutionRule=StopOldest}
}
